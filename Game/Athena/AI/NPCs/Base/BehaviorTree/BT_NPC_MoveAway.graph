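# Behavior tree for NPCs, using BB_NPC, that moves the bot to a move-away destination.
# A root Selector, gated by the move-away execution status being at least ExecutionPending, aborts itself when that condition fails.
# It runs the BT_Phoebe_UnstuckNew subtree first, falling back to a bot MoveTo on the Destination key.
# The move allows strafing, rejects partial paths and reports to the MovementState key.
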BehaviorTree /Game/Athena/AI/NPCs/Base/BehaviorTree/BT_NPC_MoveAway
blackboard=&/BB_NPC
root {
 name=Selector
 class=BTComposite_Selector
 children [
  Run Unstuck Behavior {
   class=BTTask_RunBehavior
   decorators []
   properties {behaviorAsset=&BT_Phoebe_UnstuckNew, nodeName=Run Unstuck Behavior}
  }
  MoveTo away location {
   class=FortAthenaBTTask_BotMoveTo
   decorators []
   properties {
    movementResultKeyName=AIEvaluator_MoveAway_MovementState
    nodeName=MoveTo away location
    acceptableRadius {defaultValue=30}
    allowRandomWobble {defaultValue=false}
    bAllowPartialPath {defaultValue=false}
    bAllowStrafe {defaultValue=true}
    bProjectGoalLocation {defaultValue=false}
    bReachTestIncludesAgentRadius {defaultValue=false}
    bReachTestIncludesGoalRadius {defaultValue=false}
    bTrackMovingGoal {defaultValue=false}
    blackboardKey {selectedKeyName=AIEvaluator_MoveAway_Destination}
    observedBlackboardValueTolerance {defaultValue=50}
   }
  }
 ]
 decorators [Blackboard Based Condition {
   class=BTDecorator_Blackboard
   properties {
    arithmeticOperation=GreaterOrEqual
    flowAbortMode=Self
    intValue=3
    nodeName=Blackboard Based Condition
    stringValue=ExecutionPending
    blackboardKey {selectedKeyName=AIEvaluator_MoveAway_ExecutionStatus}
   }
  }]
 properties {
  children [
   {childComposite=null, childTask=&self:BTTask_RunBehavior_1, decorators []}
   {childComposite=null, childTask=&self:FortAthenaBTTask_BotMoveTo_2, decorators []}
  ]
  services []
 }
 services []
}
